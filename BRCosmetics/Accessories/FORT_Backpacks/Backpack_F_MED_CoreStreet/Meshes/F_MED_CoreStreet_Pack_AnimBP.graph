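# Anim Blueprint for the CoreStreet backpack cosmetic, adding secondary physics motion to its dangling cords and pedal.
# The input pose passes through seven chained single-bone Anim Dynamics nodes, six cords then the pedal, simulated root-relative with per-bone gravity, damping and angular limits.
# The update event sets PedalAlpha lower during skydiving, but no node shown reads it and every dynamics alpha is fixed at full.
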AnimBlueprint /BRCosmetics/Accessories/FORT_Backpacks/Backpack_F_MED_CoreStreet/Meshes/F_MED_CoreStreet_Pack_AnimBP
target_skeleton=&/F_MED_CoreStreet_Pack_Skeleton
graphs [
 AnimGraph {nodes [
   n1 = Anim Dynamics - Bone: dyn_cord_01(ComponentPose=Local To Component(LocalPose=Input Pose()), Alpha=1, ExternalForce="0,0,0") {
    angularDampingOverride=0.95
    bOverrideAngularDamping=true
    bUseGravityOverride=true
    simulationSpace=RootRelative
    alphaBoolBlend {blendOption=Linear}
    boundBone {boneName=dyn_cord_01}
    componentAppliedLinearAccClamp=<4000 4000 4000>
    componentLinearAccScale=<0.6 0.6 0.6>
    componentLinearVelScale=<0.6 0.6 0.6>
    gravityOverride {z=-200}
    physicsBodyDefinitions [{
      boundBone {boneName=dyn_cord_01}
      boxExtents=<10 10 10>
      constraintSetup {angularLimitsMax=<0 30 15>, angularLimitsMin=<0 -30 -15>, angularTarget=<0 0 0>}
      localJointOffset=<2 0 -0.2>
     }]
   }
   n2 = Anim Dynamics - Bone: dyn_cord_02(ComponentPose=@n1, Alpha=1, ExternalForce="0,0,0") {
    angularDampingOverride=0.95
    angularSpringConstant=5000
    bGravityOverrideInSimSpace=true
    bOverrideAngularDamping=true
    bUseGravityOverride=true
    simulationSpace=RootRelative
    alphaBoolBlend {blendOption=Linear}
    boundBone {boneName=dyn_cord_02}
    componentAppliedLinearAccClamp=<700 700 1500>
    componentLinearAccScale=<0.15 0.15 0.2>
    componentLinearVelScale=<0.15 0.15 0.2>
    gravityOverride=<50 -50 -300>
    physicsBodyDefinitions [{
      boundBone {boneName=dyn_cord_02}
      boxExtents=<10 10 10>
      constraintSetup {angularLimitsMax=<0 5 25>, angularLimitsMin=<0 -15 -25>, angularTarget=<0 0 0>}
      localJointOffset=<2 0 0>
     }]
   }
   n3 = Anim Dynamics - Bone: dyn_cord_03(ComponentPose=@n2, Alpha=1, ExternalForce="0,0,0") {
    angularDampingOverride=0.95
    bOverrideAngularDamping=true
    bUseGravityOverride=true
    simulationSpace=RootRelative
    alphaBoolBlend {blendOption=Linear}
    boundBone {boneName=dyn_cord_03}
    componentAppliedLinearAccClamp=<4000 4000 4000>
    componentLinearAccScale=<0.6 0.6 0.6>
    componentLinearVelScale=<0.6 0.6 0.6>
    gravityOverride {z=-200}
    physicsBodyDefinitions [{
      boundBone {boneName=dyn_cord_03}
      boxExtents=<10 10 10>
      constraintSetup {angularLimitsMax=<0 0 20>, angularLimitsMin=<0 -15 -20>, angularTarget=<0 0 0>}
      localJointOffset=<2 0.2 0>
     }]
   }
   n4 = Anim Dynamics - Bone: dyn_cord_06(ComponentPose=@n3, Alpha=1, ExternalForce="0,0,0") {
    angularDampingOverride=0.95
    angularSpringConstant=5000
    bAngularSpring=true
    bOverrideAngularDamping=true
    bUseGravityOverride=true
    simulationSpace=RootRelative
    alphaBoolBlend {blendOption=Linear}
    boundBone {boneName=dyn_cord_06}
    componentAppliedLinearAccClamp=<4000 4000 4000>
    componentLinearAccScale=<0.6 0.6 0.6>
    componentLinearVelScale=<0.6 0.6 0.6>
    gravityOverride {z=-200}
    physicsBodyDefinitions [{
      boundBone {boneName=dyn_cord_06}
      boxExtents=<10 10 10>
      constraintSetup {angularLimitsMax=<0 5 20>, angularLimitsMin=<0 -5 -20>, angularTarget=<0 0 0>}
      localJointOffset=<2 0.2 0>
     }]
   }
   n5 = Anim Dynamics - Bone: dyn_cord_04(ComponentPose=@n4, Alpha=1, ExternalForce="0,0,0") {
    angularDampingOverride=0.999
    bOverrideAngularDamping=true
    bUseGravityOverride=true
    simulationSpace=RootRelative
    alphaBoolBlend {blendOption=Linear}
    boundBone {boneName=dyn_cord_04}
    componentAppliedLinearAccClamp=<1000 2000 2000>
    componentLinearAccScale=<0.1 0.3 0.2>
    componentLinearVelScale=<0.1 0.3 0.2>
    gravityOverride {z=-1300}
    physicsBodyDefinitions [{
      boundBone {boneName=dyn_cord_04}
      boxExtents=<17 10 10>
      constraintSetup {angularLimitsMax=<0 35 5>, angularLimitsMin=<0 0 -5>, angularTarget=<0 0 0>}
      localJointOffset=<2 0 0>
     }]
   }
   n6 = Anim Dynamics - Bone: dyn_cord_05(ComponentPose=@n5, Alpha=1, ExternalForce="0,0,0") {
    angularDampingOverride=0.95
    bOverrideAngularDamping=true
    bUseGravityOverride=true
    simulationSpace=RootRelative
    alphaBoolBlend {blendOption=Linear}
    boundBone {boneName=dyn_cord_05}
    componentAppliedLinearAccClamp=<1500 3000 3000>
    componentLinearAccScale=<0.3 0.3 0.3>
    componentLinearVelScale=<0.3 0.3 0.3>
    gravityOverride {z=-900}
    physicsBodyDefinitions [{
      boundBone {boneName=dyn_cord_05}
      boxExtents=<17 10 10>
      constraintSetup {angularLimitsMax=<0 30 0>, angularLimitsMin=<0 0 0>, angularTarget=<0 0 0>}
      localJointOffset=<2 0 0>
     }]
   }
   n7 = Anim Dynamics - Bone: dyn_pedal(ComponentPose=@n6, Alpha=1, ExternalForce="0,0,0") {
    angularDampingOverride=0.999
    bOverrideAngularDamping=true
    bUseGravityOverride=true
    simulationSpace=RootRelative
    alphaBoolBlend {blendOption=Linear}
    boundBone {boneName=dyn_pedal}
    componentAppliedLinearAccClamp=<1500 2500 2000>
    componentLinearAccScale=<0.1 0.4 0.3>
    componentLinearVelScale=<0.1 0.4 0.3>
    gravityOverride {z=-1000}
    physicsBodyDefinitions [{
      boundBone {boneName=dyn_pedal}
      boxExtents=<14 10 10>
      constraintSetup {angularLimitsMax=<0 45 20>, angularLimitsMin=<0 0 -20>, angularTarget=<0 0 0>}
      localJointOffset=<2 0 0>
     }]
   }
   Output Pose(Result=Component To Local(ComponentPose=@n7)) {layerGroup=None}
  ]}
 EventGraph {nodes [
   Event BlueprintUpdateAnimation()
   > Set PedalAlpha(PedalAlpha:double=SelectFloat(A=0.2, B=1, bPickA=Get bIsSkyDiving())) -> (Output_Get:double)
   TryGetPawnOwner()
  ]}
]
variables [PedalAlpha {type=number}]
